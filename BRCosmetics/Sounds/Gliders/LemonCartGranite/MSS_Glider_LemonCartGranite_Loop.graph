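# Looping stereo MetaSound source for the LemonCartGranite glider, layering thrust audio and music.
# Five Glider_ThrustLoop_Direction nodes feed one mixer: an AnyDirection-driven thrust loop, a music loop with no direction input, and forward, backward and sideways start one-shots.
# The music channel's gain follows the CB_LicensedAudio_Mute modulator, and a glider-open envelope follower modulates overall volume.
# Thrust.Volume and Thrust.MusicVolume inputs are unconnected; directional loop arrays are empty.
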MetaSoundSource /BRCosmetics/Sounds/Gliders/LemonCartGranite/MSS_Glider_LemonCartGranite_Loop
inputs [
 Backward.Loop.Sounds {type=WaveAsset:Array}
 Backward.Sounds {type=WaveAsset:Array}
 Forward.Loop.Sounds {type=WaveAsset:Array}
 Forward.Sounds {type=WaveAsset:Array}
 Glider.AnyDirection {type=Float}
 Glider.Backwards {type=Float}
 Glider.Forwards {type=Float}
 Glider.Sideways {type=Float}
 Sideways.Loop.Sounds {type=WaveAsset:Array}
 Sideways.Sounds {type=WaveAsset:Array}
 Thrust.Loop.Music {type=WaveAsset:Array}
 Thrust.Loop.Sounds {type=WaveAsset:Array}
 Thrust.MusicVolume {type=Float}
 Thrust.OneShotVolume {type=Float}
 Thrust.Sounds {type=WaveAsset:Array}
 Thrust.Volume {type=Float}
 UE.Source.OnPlay {type=Trigger}
]
interfaces [Fort.Glider:1.0, UE.OutputFormat.Stereo:1.0, UE.Source:1.1]
outputs [
 "UE.OutputFormat.Stereo.Audio:0" {type=Audio}
 "UE.OutputFormat.Stereo.Audio:1" {type=Audio}
]
pages [Default {
  input_defaults {
   Glider.AnyDirection=0
   Glider.Backwards=0
   Glider.Forwards=0
   Glider.Sideways=0
   Thrust.MusicVolume=1
   Thrust.OneShotVolume=1
   Thrust.Volume=1.5
   UE.Source.OnPlay=null
   Backward.Loop.Sounds []
   Backward.Sounds [
    &/IP_Glider_LemonCartGranite_ThrustStart_Small_01
    &/IP_Glider_LemonCartGranite_ThrustStart_Small_02
    &/IP_Glider_LemonCartGranite_ThrustStart_Small_03
   ]
   Forward.Loop.Sounds []
   Forward.Sounds [
    &/IP_Glider_LemonCartGranite_ThrustStart_01
    &/IP_Glider_LemonCartGranite_ThrustStart_02
    &/IP_Glider_LemonCartGranite_ThrustStart_03
   ]
   Sideways.Loop.Sounds []
   Sideways.Sounds [
    &IP_Glider_LemonCartGranite_ThrustStart_Small_01
    &IP_Glider_LemonCartGranite_ThrustStart_Small_02
    &IP_Glider_LemonCartGranite_ThrustStart_Small_03
   ]
   Thrust.Loop.Music [&/IP_Glider_LemonCartGranite_Music]
   Thrust.Loop.Sounds [&/IP_Glider_LemonCartGranite_Forward]
   Thrust.Sounds []
  }
  nodes [
   n1 = /FortMetasound/Cosmetics/Glider/Glider_ThrustLoop_Direction(Value=^Glider.AnyDirection, Loop=^Thrust.Loop.Sounds, On Play=^UE.Source.OnPlay) {also {defaults {PitchMax=1.1, PitchMin=1, VolumeMax=0.3, VolumeMin=0.3}}}
   n2 = &Glider_ThrustLoop_Direction(Loop=^Forward.Loop.Sounds, Oneshots=^Forward.Sounds, Value=^Glider.Forwards, OneShotVolume=^Thrust.OneShotVolume, On Play=^UE.Source.OnPlay) {also {defaults {PitchMax=1, PitchMin=1}}}
   n3 = &Glider_ThrustLoop_Direction(Loop=^Thrust.Loop.Music, On Play=^UE.Source.OnPlay) {also {defaults {PitchMax=1, PitchMin=1, VolumeMax=0.3, VolumeMin=0.3}}}
   n4 = &Glider_ThrustLoop_Direction(
    Loop=^Backward.Loop.Sounds
    Oneshots=^Backward.Sounds
    Value=^Glider.Backwards
    OneShotVolume=^Thrust.OneShotVolume
    On Play=^UE.Source.OnPlay
   ) {also {defaults {PitchMax=1, PitchMin=1}}}
   n5 = &Glider_ThrustLoop_Direction(
    Value=^Glider.Sideways
    Loop=^Sideways.Loop.Sounds
    Oneshots=^Sideways.Sounds
    OneShotVolume=^Thrust.OneShotVolume
    On Play=^UE.Source.OnPlay
   ) {also {defaults {PitchMax=1, PitchMin=1}}}
   AudioMixerAudioMixerStereo5_0 = "AudioMixer.Audio Mixer (Stereo, 5)"(
    In 4 L=@n1.Out L
    In 4 R=@n1.Out R
    In 0 L=@n2.Out L
    In 0 R=@n2.Out R
    In 3 L=@n3.Out L
    In 3 R=@n3.Out R
    In 2 L=@n4.Out L
    In 2 R=@n4.Out R
    In 1 L=@n5.Out L
    In 1 R=@n5.Out R
    Gain 3=Modulation.GetModulatorValue() {also {defaults {Modulator=/Game/Sounds/Modulation/Parameters/CB_LicensedAudio_Mute}}}.Out
   ) {also {defaults {Gain 0=0.2, Gain 1=0.2, Gain 2=0.2}}}
  ]
  links [
   {dst="@UE.OutputFormat.Stereo.Audio:0", src=AudioMixerAudioMixerStereo5_0.Out L}
   {dst="@UE.OutputFormat.Stereo.Audio:1", src=AudioMixerAudioMixerStereo5_0.Out R}
  ]
 }]
properties {
 assetImportData=&self:AssetImportData
 attenuationSettings=&Fort_Player_Glider_Thrust_Athena_Att
 audioPropertiesSheet=&APS_Cosmetic_Glider_Loop
 bLooping=true
 duration=10000
 maxDistance=12500
 numChannels=2
 outputFormat=Stereo
 sampleRate=48000
 soundClassObject=/Game/Sounds/FortSoundClassesAndModes/Fort_Gliders_Music
 volume=1.2
 assetUserData [&self:AudioAssetUserData_0]
 concurrencySet [&Athena_Glider_Thrust_Conc]
 modulationSettings {
  volumeRouting=Union
  volumeModulationDestination {bEnableModulation=true, modulators [/Game/Sounds/Modulation/EnvelopeFollowers/Volume/CB_Glider_Open_EnvFollower]}
 }
}
